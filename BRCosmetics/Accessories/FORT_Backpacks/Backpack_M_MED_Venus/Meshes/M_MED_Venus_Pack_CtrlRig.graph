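# Control Rig for the Venus backpack cosmetic that procedurally animates its left and right plant-like stems and petals.
# Bone harmonics sway the upper stems, while chain harmonics with wave and pendulum drive the lower stem chains.
# Transform constraints make the lower petal masters follow the chain tips with maintained offset.
# Additive local rotations flutter each petal pair, weighted by sine timelines with randomized speed; four SetBoneRotation nodes are left unconnected.
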ControlRigBlueprint /BRCosmetics/Accessories/FORT_Backpacks/Backpack_M_MED_Venus/Meshes/M_MED_Venus_Pack_CtrlRig
graphs [
 RigVMFunctionLibrary {links [], nodes []}
 RigVMModel {nodes [
   RigUnit_BeginExecution = &RigUnit_BeginExecution()
   /Script/ControlRig.RigUnit_BoneHarmonics(
    Bones=((Bone="dyn_stem_upr_r_01"))
    RotationOrder=ZYX
    WaveAmplitude=(X=15,Y=15,Z=30)
    WaveFrequency=(X=1,Y=0.6,Z=0.8)
    WaveMaximum=1
    WaveMinimum=0.5
    WaveNoise=(X=0,Y=0.2,Z=0.2)
    WaveOffset=(X=0,Y=1,Z=3)
    WaveSpeed=(X=1,Y=1,Z=1)
    bPropagateToChildren=True
    ExecutePin=@RigUnit_BeginExecution.ExecutePin
   ) {is=struct RigVMUnitNode}
   > &RigUnit_BoneHarmonics(
    Bones=((Bone="dyn_stem_upr_l_01"))
    RotationOrder=ZYX
    WaveAmplitude=(X=15,Y=15,Z=30)
    WaveFrequency=(X=1,Y=0.6,Z=0.8)
    WaveMaximum=1
    WaveMinimum=0.5
    WaveNoise=(X=0,Y=0.3,Z=0.3)
    WaveOffset=(X=0,Y=2,Z=5)
    WaveSpeed=(X=1,Y=1,Z=1)
    bPropagateToChildren=True
   )
   > /Script/ControlRig.RigUnit_ChainHarmonics(
    ChainRoot=dyn_stem_lwr_r_01
    DrawWorldOffset=(Rotation=(X=0,Y=0,Z=0,W=1),Translation=(X=0,Y=0,Z=0),Scale3D=(X=1,Y=1,Z=1))
    Pendulum=(bEnabled=True,PendulumStiffness=2,PendulumGravity=(X=0,Y=0,Z=0),PendulumBlend=0.75,PendulumDrag=0.98,PendulumMaximum=0.1,UnwindAxis=(X=0,Y=1,Z=0),UnwindMinimum=0.2,UnwindMaximum=0.05)
    Reach=(bEnabled=True,ReachTarget=(X=0,Y=0,Z=0),ReachAxis=(X=1,Y=0,Z=0))
    Speed=(X=0.8,Y=1,Z=0.8)
    Wave=(bEnabled=True,WaveFrequency=(X=1,Y=1,Z=1),WaveAmplitude=(X=0.5,Y=0.8,Z=1),WaveOffset=(X=0,Y=0,Z=0.3),WaveNoise=(X=0.5,Y=0.5,Z=0.5),WaveMinimum=0.4,WaveMaximum=1,WaveEase=SineEaseOut)
    WaveCurve="(EditorCurveData=(Keys=((InterpMode=RCIM_Cubic,TangentMode=RCTM_User,Value=0.2,ArriveTangent=594,LeaveTangent=0.2),(InterpMode=RCIM_Cubic,Time=1,Value=0.3,LeaveTangent=0.3962)),DefaultValue=340282346638528859811704183484516925440,PreInfinityExtrap=RCCE_Con ...(truncated)"
    bDrawDebug=True
   ) {is=struct RigVMUnitNode}
   > &RigUnit_ChainHarmonics(
    ChainRoot=dyn_stem_lwr_l_01
    DrawWorldOffset=(Rotation=(X=0,Y=0,Z=0,W=1),Translation=(X=0,Y=0,Z=0),Scale3D=(X=1,Y=1,Z=1))
    Pendulum=(bEnabled=True,PendulumStiffness=2,PendulumGravity=(X=0,Y=0,Z=0),PendulumBlend=0.75,PendulumDrag=0.98,PendulumMaximum=0.1,UnwindAxis=(X=0,Y=1,Z=0),UnwindMinimum=0.2,UnwindMaximum=0.05)
    Reach=(bEnabled=True,ReachTarget=(X=0,Y=0,Z=0),ReachAxis=(X=1,Y=0,Z=0))
    Speed=(X=0.8,Y=1,Z=0.8)
    Wave=(bEnabled=True,WaveFrequency=(X=0.1,Y=0.7,Z=0.9),WaveAmplitude=(X=0.5,Y=0.8,Z=1),WaveOffset=(X=0,Y=-1,Z=5),WaveNoise=(X=0.3,Y=0.5,Z=1),WaveMinimum=0.4,WaveMaximum=1,WaveEase=SineEaseOut)
    WaveCurve="(EditorCurveData=(Keys=((InterpMode=RCIM_Cubic,TangentMode=RCTM_User,Value=0.2,ArriveTangent=1098.4282,LeaveTangent=0.3097),(InterpMode=RCIM_Cubic,Time=1,Value=0.3,LeaveTangent=0.3962)),DefaultValue=340282346638528859811704183484516925440,PreInfinityExtrap ...(truncated)"
    bDrawDebug=True
   )
   > /Script/ControlRig.RigUnit_TransformConstraint(
    BaseBone=None
    BaseTransform=(Rotation=(X=0,Y=0,Z=0,W=1),Translation=(X=0,Y=0,Z=0),Scale3D=(X=1,Y=1,Z=1))
    BaseTransformSpace=GlobalSpace
    Bone=dyn_petal_lwr_r_MSTR
    Targets=((Transform=(Rotation=(X=0,Y=0,Z=0,W=1),Translation=(X=0,Y=0,Z=0),Scale3D=(X=1,Y=1,Z=1)),Weight=1,bMaintainOffset=True,Filter=(TranslationFilter=(bX=True,bY=True,bZ=True),RotationFilter=(bX=True,bY=True,bZ=True),ScaleFilter=(bX=True,bY=True,bZ=True))))
    bUseInitialTransforms=True
    Targets.0.Transform=&RigUnit_GetBoneTransform(Bone=dyn_stem_lwr_r_04).Transform
   ) {is=struct RigVMUnitNode}
   > &RigUnit_TransformConstraint(
    BaseBone=None
    BaseTransform=(Rotation=(X=0,Y=0,Z=0,W=1),Translation=(X=0,Y=0,Z=0),Scale3D=(X=1,Y=1,Z=1))
    BaseTransformSpace=GlobalSpace
    Bone=dyn_stem_lwr_l_MSTR
    Targets=((Transform=(Rotation=(X=0,Y=0,Z=0,W=1),Translation=(X=0,Y=0,Z=0),Scale3D=(X=1,Y=1,Z=1)),Weight=1,bMaintainOffset=True,Filter=(TranslationFilter=(bX=True,bY=True,bZ=True),RotationFilter=(bX=True,bY=True,bZ=True),ScaleFilter=(bX=True,bY=True,bZ=True))))
    bUseInitialTransforms=True
    Targets.0.Transform=&RigUnit_GetBoneTransform(Bone=dyn_stem_lwr_l_04).Transform
   )
   n1 = /Script/RigVM.RigVMFunction_RandomFloat(Duration=1, Maximum=2.5, Minimum=1, Seed=500) {is=struct RigVMUnitNode, template="Random::Execute(in Seed,in Minimum,in Maximum,in Duration,out Result)"}
   n2 = &RigVMFunction_MathQuaternionFromEuler(Euler=(X=2,Y=20,Z=0))
   > /Script/ControlRig.RigUnit_ModifyBoneTransforms(
    BoneToModify=((Bone="dyn_petal_upr_r_01",Transform=(Rotation=(X=0,Y=0,Z=0,W=1),Translation=(X=0,Y=0,Z=0),Scale3D=(X=1,Y=1,Z=1))),(Bone="dyn_petal_upr_r_02",Transform=(Rotation=(X=0,Y=0,Z=0,W=1),Translation=(X=0,Y=0,Z=0),Scale3D=(X=1,Y=1,Z=1))))
    Mode=AdditiveLocal
    WeightMaximum=1
    WeightMinimum=0
    Weight=/Script/RigVM.RigVMFunction_MathFloatSin(Value=/Script/RigVM.RigVMFunction_Timeline(Speed=@n1.Result) {is=struct RigVMUnitNode}.Time) {is=struct RigVMUnitNode, template="Sin::Execute(in Value,out Result)"}.Result
    BoneToModify.0.Transform.Rotation=@n2.Result
    BoneToModify.1.Transform.Rotation=@n2.Result
   ) {is=struct RigVMUnitNode}
   n3 = &RigVMFunction_RandomFloat(Duration=2, Maximum=2, Minimum=1, Seed=300) {template="Random::Execute(in Seed,in Minimum,in Maximum,in Duration,out Result)"}
   n4 = &RigVMFunction_MathQuaternionFromEuler(Euler=(X=2,Y=20,Z=0))
   > &RigUnit_ModifyBoneTransforms(
    BoneToModify=((Bone="dyn_petal_upr_l_01",Transform=(Rotation=(X=0,Y=0,Z=0,W=1),Translation=(X=0,Y=0,Z=0),Scale3D=(X=1,Y=1,Z=1))),(Bone="dyn_petal_upr_l_02",Transform=(Rotation=(X=0,Y=0,Z=0,W=1),Translation=(X=0,Y=0,Z=0),Scale3D=(X=1,Y=1,Z=1))))
    Mode=AdditiveLocal
    WeightMaximum=1
    WeightMinimum=0
    Weight=&RigVMFunction_MathFloatSin(Value=&RigVMFunction_Timeline(Speed=@n3.Result).Time) {template="Sin::Execute(in Value,out Result)"}.Result
    BoneToModify.0.Transform.Rotation=@n4.Result
    BoneToModify.1.Transform.Rotation=@n4.Result
   )
   n5 = &RigVMFunction_RandomFloat(Duration=2, Maximum=3, Minimum=1, Seed=500) {template="Random::Execute(in Seed,in Minimum,in Maximum,in Duration,out Result)"}
   n6 = &RigVMFunction_MathQuaternionFromEuler(Euler=(X=0,Y=15,Z=-5))
   > &RigUnit_ModifyBoneTransforms(
    BoneToModify=((Bone="dyn_petal_lwr_r_01",Transform=(Rotation=(X=0,Y=0,Z=0,W=1),Translation=(X=0,Y=0,Z=0),Scale3D=(X=1,Y=1,Z=1))),(Bone="dyn_petal_lwr_r_02",Transform=(Rotation=(X=0,Y=0,Z=0,W=1),Translation=(X=0,Y=0,Z=0),Scale3D=(X=1,Y=1,Z=1))))
    Mode=AdditiveLocal
    WeightMaximum=1
    WeightMinimum=0
    Weight=&RigVMFunction_MathFloatSin(Value=&RigVMFunction_Timeline(Speed=@n5.Result).Time) {template="Sin::Execute(in Value,out Result)"}.Result
    BoneToModify.0.Transform.Rotation=@n6.Result
    BoneToModify.1.Transform.Rotation=@n6.Result
   )
   n7 = &RigVMFunction_RandomFloat(Duration=1, Maximum=3, Minimum=1, Seed=300) {template="Random::Execute(in Seed,in Minimum,in Maximum,in Duration,out Result)"}
   n8 = &RigVMFunction_MathQuaternionFromEuler(Euler=(X=0,Y=15,Z=5))
   > &RigUnit_ModifyBoneTransforms(
    BoneToModify=((Bone="dyn_petal_lwr_l_01",Transform=(Rotation=(X=0,Y=0,Z=0,W=1),Translation=(X=0,Y=0,Z=0),Scale3D=(X=1,Y=1,Z=1))),(Bone="dyn_petal_lwr_l_02",Transform=(Rotation=(X=0,Y=0,Z=0,W=1),Translation=(X=0,Y=0,Z=0),Scale3D=(X=1,Y=1,Z=1))))
    Mode=AdditiveLocal
    WeightMaximum=1
    WeightMinimum=0
    Weight=&RigVMFunction_MathFloatSin(Value=&RigVMFunction_Timeline(Speed=@n7.Result).Time) {template="Sin::Execute(in Value,out Result)"}.Result
    BoneToModify.0.Transform.Rotation=@n8.Result
    BoneToModify.1.Transform.Rotation=@n8.Result
   )
   RigUnit_SetBoneRotation = /Script/ControlRig.RigUnit_SetBoneRotation(Bone=dyn_petal_upr_r_01, ExecutePin=(), Rotation=(X=0,Y=0,Z=0,W=1), Weight=1, bPropagateToChildren=True) {is=struct RigVMUnitNode}
   &RigUnit_SetBoneRotation(Bone=dyn_petal_upr_r_02, ExecutePin=(), Rotation=(X=0,Y=0,Z=0,W=1), Weight=1, bPropagateToChildren=True)
   &RigUnit_SetBoneRotation(Bone=dyn_petal_upr_l_01, ExecutePin=(), Rotation=(X=0,Y=0,Z=0,W=1), Weight=1, bPropagateToChildren=True)
   &RigUnit_SetBoneRotation(Bone=dyn_petal_upr_l_02, ExecutePin=(), Rotation=(X=0,Y=0,Z=0,W=1), Weight=1, bPropagateToChildren=True)
  ]}
]
hierarchy [
 root {loc=<0 0 0>, rot=<0 0 0 1>}
 pack {parent=root, loc=<0 -0.7316 0>, rot=<0 0 0 1>}
 dyn_body {parent=pack, loc=<0 -1.4347 -3.7878>, rot=<0 0 0 1>}
 dyn_stem_upr_r_01 {parent=dyn_body, loc=<-3.778 0.0639 2.3391>, rot=<0.1073 0.0893 0.9549 -0.2621>}
 dyn_petal_upr_r_01_MSTR {parent=dyn_stem_upr_r_01, loc=<4.4752 0.7487 1.6296>, rot=<0.5092 0.1851 -0.3445 -0.7667>}
 dyn_petal_upr_r_01 {parent=dyn_petal_upr_r_01_MSTR, loc=<0 0 0>, rot=<0 0 0 1>}
 dyn_petal_upr_r_02_MSTR {parent=dyn_stem_upr_r_01, loc=<5.6124 -2.0062 0.3772>, rot=<-0.7482 -0.1885 0.1005 -0.6282>}
 dyn_petal_upr_r_02 {parent=dyn_petal_upr_r_02_MSTR, loc=<0 0 0>, rot=<0 0 0 -1>}
 dyn_stem_lwr_r_01 {parent=dyn_body, loc=<-5.968 -1.4312 -6.3228>, rot=<-0.3351 -0.3327 0.6209 -0.6257>}
 dyn_stem_lwr_r_02 {parent=dyn_stem_lwr_r_01, loc=<8.5064 0.365 -0.2866>, rot=<-0.359 0.1838 0.3872 -0.8291>}
 dyn_stem_lwr_r_03 {parent=dyn_stem_lwr_r_02, loc=<7.7076 0 0.4484>, rot=<-0.0795 0.6251 -0.3694 -0.683>}
 dyn_stem_lwr_r_04 {parent=dyn_stem_lwr_r_03, loc=<7.5804 -0.1291 -0.3105>, rot=<0.0054 -0.3379 0.0666 -0.9388>}
 dyn_stem_upr_l_01 {parent=dyn_body, loc=<3.778 0.0639 2.3391>, rot=<0.0893 0.1073 0.2621 -0.9549>}
 dyn_petal_upr_l_01_MSTR {parent=dyn_stem_upr_l_01, loc=<4.4752 -0.7487 1.6296>, rot=<-0.5092 0.1851 0.3445 -0.7667>}
 dyn_petal_upr_l_01 {parent=dyn_petal_upr_l_01_MSTR, loc=<0 0 0>, rot=<0 0 0 -1>}
 dyn_petal_upr_l_02_MSTR {parent=dyn_stem_upr_l_01, loc=<5.6124 2.0062 0.3772>, rot=<0.7482 -0.1885 -0.1005 -0.6282>}
 dyn_petal_upr_l_02 {parent=dyn_petal_upr_l_02_MSTR, loc=<0 0 0>, rot=<0 0 0 1>}
 dyn_stem_lwr_l_01 {parent=dyn_body, loc=<5.9681 -1.4312 -6.3228>, rot=<-0.3327 -0.3351 0.6257 -0.6209>}
 dyn_stem_lwr_l_02 {parent=dyn_stem_lwr_l_01, loc=<8.5064 -0.365 -0.2865>, rot=<0.359 0.1838 -0.3872 -0.8291>}
 dyn_stem_lwr_l_03 {parent=dyn_stem_lwr_l_02, loc=<7.7076 -0.0001 0.4483>, rot=<0.0795 0.6251 0.3694 -0.683>}
 dyn_stem_lwr_l_04 {parent=dyn_stem_lwr_l_03, loc=<7.5805 0.1292 -0.3104>, rot=<-0.0054 -0.3379 -0.0666 -0.9388>}
 dyn_petal_lwr_r_MSTR {parent=dyn_body, loc=<-13.3506 -13.3638 -11.2474>, rot=<0.0839 0.0907 0.8426 -0.5241>}
 dyn_petal_lwr_r_01_MSTR {parent=dyn_petal_lwr_r_MSTR, loc=<2.6254 0.8376 -0.2245>, rot=<0.6901 0.1752 -0.2348 -0.6617>}
 dyn_petal_lwr_r_01 {parent=dyn_petal_lwr_r_01_MSTR, loc=<0 0 0>, rot=<0 0 0 -1>}
 dyn_petal_lwr_r_02_MSTR {parent=dyn_petal_lwr_r_MSTR, loc=<2.7518 -1.1622 -0.1318>, rot=<-0.6671 0.1144 0.1538 -0.7199>}
 dyn_petal_lwr_r_02 {parent=dyn_petal_lwr_r_02_MSTR, loc=<0 0 0>, rot=<0 0 0 -1>}
 dyn_stem_lwr_l_MSTR {parent=dyn_body, loc=<13.3506 -13.3638 -11.2474>, rot=<0.0907 0.0839 0.5241 -0.8426>}
 dyn_petal_lwr_l_01_MSTR {parent=dyn_stem_lwr_l_MSTR, loc=<2.6254 -0.8376 -0.2246>, rot=<-0.6901 0.1752 0.2348 -0.6617>}
 dyn_petal_lwr_l_01 {parent=dyn_petal_lwr_l_01_MSTR, loc=<0 0 0>, rot=<0 0 0 1>}
 dyn_petal_lwr_l_02_MSTR {parent=dyn_stem_lwr_l_MSTR, loc=<2.7518 1.1622 -0.1318>, rot=<0.6671 0.1144 -0.1538 -0.7199>}
 dyn_petal_lwr_l_02 {parent=dyn_petal_lwr_l_02_MSTR, loc=<0 0 0>, rot=<0 0 0 -1>}
]
