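# Physics asset for the Venus backpack cosmetic, simulating secondary motion on its plant-like stems.
# Small capsule bodies on dyn_body anchor two four-segment lower stem chains and single upper stem segments on left and right.
# Stem joints use tight swing and twist limits, stiffest at the lower stem roots.
# The lower right petal and lower left stem MSTR bones attach to the body with fully free angular motion.
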PhysicsAsset /BRCosmetics/Accessories/FORT_Backpacks/Backpack_M_MED_Venus/Meshes/M_MED_Venus_Pack_Physics
bodies {
 dyn_body [{shapeType=Capsule, length=0.5, radius=0.5}]
 dyn_petal_lwr_r_MSTR [{shapeType=Capsule, length=0.5, radius=0.5}]
 dyn_stem_lwr_l_01 [{shapeType=Capsule, length=3, radius=1.5, center=<0.1 0 0>, rotation=<rot 90 0 0>}]
 dyn_stem_lwr_l_02 [{shapeType=Capsule, length=3, radius=2, center=<0.1 0 0>, rotation=<rot 90 0 0>}]
 dyn_stem_lwr_l_03 [{shapeType=Capsule, length=4, radius=1.5, center=<0.1 0 0>, rotation=<rot 90 0 0>}]
 dyn_stem_lwr_l_04 [{shapeType=Capsule, length=4, radius=1.5, center=<0.1 0 0>, rotation=<rot 90 0 0>}]
 dyn_stem_lwr_l_MSTR [{shapeType=Capsule, length=0.5, radius=0.5}]
 dyn_stem_lwr_r_01 [{shapeType=Capsule, length=3, radius=1.5, rotation=<rot 90 0 0>}]
 dyn_stem_lwr_r_02 [{shapeType=Capsule, length=3, radius=2, rotation=<rot 90 0 0>}]
 dyn_stem_lwr_r_03 [{shapeType=Capsule, length=4, radius=1.5, rotation=<rot 90 0 0>}]
 dyn_stem_lwr_r_04 [{shapeType=Capsule, length=4, radius=1.5, rotation=<rot 90 0 0>}]
 dyn_stem_upr_l_01 [{shapeType=Capsule, length=3, radius=1.5, center=<1.5 0 0>, rotation=<rot 90 0 0>}]
 dyn_stem_upr_r_01 [{shapeType=Capsule, length=3, radius=1.5, center=<1.5 0 0>, rotation=<rot 90 0 0>}]
}
constraints [
 {bone1Name=dyn_stem_upr_r_01, bone2Name=dyn_body, swing1LimitDegrees=10, swing2LimitDegrees=15, twistLimitDegrees=5}
 {bone1Name=dyn_stem_lwr_r_01, bone2Name=dyn_body, swing1LimitDegrees=5, swing2LimitDegrees=5, twistLimitDegrees=2}
 {bone1Name=dyn_stem_lwr_r_02, bone2Name=dyn_stem_lwr_r_01, swing1LimitDegrees=15, swing2LimitDegrees=15, twistLimitDegrees=5}
 {bone1Name=dyn_stem_lwr_r_03, bone2Name=dyn_stem_lwr_r_02, swing1LimitDegrees=15, swing2LimitDegrees=15, twistLimitDegrees=5}
 {bone1Name=dyn_stem_lwr_r_04, bone2Name=dyn_stem_lwr_r_03, swing1LimitDegrees=10, swing2LimitDegrees=10, twistLimitDegrees=5}
 {bone1Name=dyn_stem_lwr_l_01, bone2Name=dyn_body, swing1LimitDegrees=5, swing2LimitDegrees=5, twistLimitDegrees=2}
 {bone1Name=dyn_stem_lwr_l_02, bone2Name=dyn_stem_lwr_l_01, swing1LimitDegrees=15, swing2LimitDegrees=15, twistLimitDegrees=5}
 {bone1Name=dyn_stem_lwr_l_03, bone2Name=dyn_stem_lwr_l_02, swing1LimitDegrees=15, swing2LimitDegrees=15, twistLimitDegrees=5}
 {bone1Name=dyn_stem_lwr_l_04, bone2Name=dyn_stem_lwr_l_03, swing1LimitDegrees=15, swing2LimitDegrees=15, twistLimitDegrees=5}
 {bone1Name=dyn_stem_upr_l_01, bone2Name=dyn_body, swing1LimitDegrees=10, swing2LimitDegrees=15, twistLimitDegrees=5}
 {
  bone1Name=dyn_petal_lwr_r_MSTR
  bone2Name=dyn_body
  swing1LimitDegrees=45
  swing1Motion=Free
  swing2LimitDegrees=45
  swing2Motion=Free
  twistLimitDegrees=45
  twistMotion=Free
 }
 {
  bone1Name=dyn_stem_lwr_l_MSTR
  bone2Name=dyn_body
  swing1LimitDegrees=45
  swing1Motion=Free
  swing2LimitDegrees=45
  swing2Motion=Free
  twistLimitDegrees=45
  twistMotion=Free
 }
]
