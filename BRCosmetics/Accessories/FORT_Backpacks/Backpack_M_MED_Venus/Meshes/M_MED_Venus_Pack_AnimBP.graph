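# Anim Blueprint for the Venus backpack cosmetic, which makes its petals wiggle when the player gets a kill.
# The graph runs the reference pose through stem/petal constraints, RigidBody physics, additive petal bone rotations and a final Control Rig.
# The Event Graph watches the pawn's AthenaKills stat; an increase starts sine-driven upper and lower petal rotations.
# Their blend alpha fades out on a counter incremented per animation update, so duration depends on frame rate.
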AnimBlueprint /BRCosmetics/Accessories/FORT_Backpacks/Backpack_M_MED_Venus/Meshes/M_MED_Venus_Pack_AnimBP
target_skeleton=&/M_MED_Venus_Pack_Skeleton
graphs [
 AnimGraph {nodes [
   Input Pose()
   n1 = "Constraint - dyn_petal_lwr_r_MSTR (P)"(ComponentPose=Local To Component(LocalPose=Local Space Ref Pose()), Alpha=1, ConstraintWeights_0=1) {
    alphaBoolBlend {blendOption=Linear}
    boneToModify {boneName=dyn_petal_lwr_r_MSTR}
    constraintSetup [{targetBone {boneName=dyn_stem_lwr_r_04}}]
    constraintWeights [1]
   }
   n2 = "Constraint - dyn_stem_lwr_l_MSTR (P)"(ComponentPose=@n1, Alpha=1, ConstraintWeights_0=1) {
    alphaBoolBlend {blendOption=Linear}
    boneToModify {boneName=dyn_stem_lwr_l_MSTR}
    constraintSetup [{targetBone {boneName=dyn_stem_lwr_l_04}}]
    constraintWeights [1]
   }
   n3 = RigidBody(ComponentPose=@n2) {
    bOverrideWorldGravity=true
    overridePhysicsAsset=&/M_MED_Venus_Pack_Physics
    simulationSpace=BaseBoneSpace
    alphaBoolBlend {blendOption=Linear}
    baseBoneRef {boneName=root}
    componentAppliedLinearAccClamp=<2000 2000 2000>
    componentLinearAccScale=<0.3 0.3 0.3>
    componentLinearVelScale=<0.2 0.2 0.2>
    overrideWorldGravity {z=-980}
   }
   n4 = Get petalUpperRotate() -> (petalUpperRotate:Rotator)
   n5 = Get didKill() -> (didKill:double)
   n6 = "Transform (Modify) Bone - Bone: dyn_petal_upr_r_01_MSTR"(ComponentPose=@n3, Alpha=@n5, Translation="0,0,0", Rotation=@n4, Scale="1,1,1") {rotationMode=BMM_Additive, rotationSpace=BCS_BoneSpace, alphaBoolBlend {blendOption=Linear}, boneToModify {boneName=dyn_petal_upr_r_01_MSTR}}
   n7 = "Transform (Modify) Bone - Bone: dyn_petal_upr_r_02_MSTR"(ComponentPose=@n6, Alpha=@n5, Translation="0,0,0", Rotation=@n4, Scale="1,1,1") {rotationMode=BMM_Additive, rotationSpace=BCS_BoneSpace, alphaBoolBlend {blendOption=Linear}, boneToModify {boneName=dyn_petal_upr_r_02_MSTR}}
   n8 = "Transform (Modify) Bone - Bone: dyn_petal_upr_l_01_MSTR"(ComponentPose=@n7, Alpha=@n5, Translation="0,0,0", Rotation=@n4, Scale="1,1,1") {rotationMode=BMM_Additive, rotationSpace=BCS_BoneSpace, alphaBoolBlend {blendOption=Linear}, boneToModify {boneName=dyn_petal_upr_l_01_MSTR}}
   n9 = "Transform (Modify) Bone - Bone: dyn_petal_upr_l_02_MSTR"(ComponentPose=@n8, Alpha=@n5, Translation="0,0,0", Rotation=@n4, Scale="1,1,1") {rotationMode=BMM_Additive, rotationSpace=BCS_BoneSpace, alphaBoolBlend {blendOption=Linear}, boneToModify {boneName=dyn_petal_upr_l_02_MSTR}}
   n10 = Get petalLowerRotate() -> (petalLowerRotate:Rotator)
   n11 = "Transform (Modify) Bone - Bone: dyn_petal_lwr_l_01_MSTR"(ComponentPose=@n9, Alpha=@n5, Translation="0,0,0", Rotation=@n10, Scale="1,1,1") {rotationMode=BMM_Additive, rotationSpace=BCS_BoneSpace, alphaBoolBlend {blendOption=Linear}, boneToModify {boneName=dyn_petal_lwr_l_01_MSTR}}
   n12 = "Transform (Modify) Bone - Bone: dyn_petal_lwr_l_02_MSTR"(ComponentPose=@n11, Alpha=@n5, Translation="0,0,0", Rotation=@n10, Scale="1,1,1") {rotationMode=BMM_Additive, rotationSpace=BCS_BoneSpace, alphaBoolBlend {blendOption=Linear}, boneToModify {boneName=dyn_petal_lwr_l_02_MSTR}}
   n13 = "Transform (Modify) Bone - Bone: dyn_petal_lwr_r_01_MSTR"(ComponentPose=@n12, Alpha=@n5, Translation="0,0,0", Rotation=@n10, Scale="1,1,1") {rotationMode=BMM_Additive, rotationSpace=BCS_BoneSpace, alphaBoolBlend {blendOption=Linear}, boneToModify {boneName=dyn_petal_lwr_r_01_MSTR}}
   n14 = "Transform (Modify) Bone - Bone: dyn_petal_lwr_r_02_MSTR"(ComponentPose=@n13, Alpha=@n5, Translation="0,0,0", Rotation=@n10, Scale="1,1,1") {rotationMode=BMM_Additive, rotationSpace=BCS_BoneSpace, alphaBoolBlend {blendOption=Linear}, boneToModify {boneName=dyn_petal_lwr_r_02_MSTR}}
   n15 = Control Rig(Source=Component To Local(ComponentPose=@n14)) {
    alphaBoolBlend {blendOption=Linear}
    controlRigAssetReference {blueprintRigClass=&/M_MED_Venus_Pack_CtrlRig.M_MED_Venus_Pack_CtrlRig_C}
   }
   Output Pose(Result=@n15) {layerGroup=None}
  ]}
 EventGraph {nodes [
   n1 = Event BlueprintUpdateAnimation()
   Get!(execute:Exec=[@n1.then, @n10.then_1]) -> (then:Exec, else:Exec, Pawn:Fort Player Pawn Object Reference)
   > Branch(Condition=Get ActiveKill() -> (ActiveKill:bool))
   > Set timer(timer:double=float + float(A=Get timer() -> (timer:double), B=1) {is=title K2Node_CommutativeAssociativeBinaryOperator}) -> (Output_Get:double)
   n2 = "Sin (Degrees)"!(A:double=float * float(A=GetGameTimeInSeconds(), B=800)) -> (ReturnValue:double)
   > Set petalUpperRotate(petalUpperRotate:Rotator=MakeRotator(Pitch=float * float(A=@n2, B=20))) -> (Output_Get:Rotator)
   n3 = "Sin (Degrees)"(A=float * float(A=GetGameTimeInSeconds(), B=600))
   > Set petalLowerRotate(petalLowerRotate:Rotator=MakeRotator(Pitch=float * float(A=@n3, B=25))) -> (Output_Get:Rotator)
   n4 = Get timer()
   n5 = Get timerLen() -> (timerLen:double)
   > Set didKill(didKill:double=MapRangeClamped(Value=float - float(A=@n5, B=@n4) {is=title K2Node_CallFunction}, InRangeA=Get timerLen(), InRangeB=0, OutRangeA=1, OutRangeB=0)) -> (Output_Get:double)
   > Branch(Condition="Equal (Float)"!(A:double=Get didKill(), B:double=0) -> (ReturnValue:bool) {is=title K2Node_CallFunction})
   > Set ActiveKill(ActiveKill:bool=false) -> (Output_Get:bool)
   > Set timer(timer=0)
   Event BlueprintInitializeAnimation()
   > n6 = Cast To FortPlayerPawn(Object=TryGetFortPawn())
   > Set Pawn(Pawn:Fort Player Pawn Object Reference=@n6) -> (Output_Get:Fort Player Pawn Object Reference)
   n7 = Event OnCharacterCustomizationLoaded()
   n8 = WatchCosmeticStat!(
    execute:Exec=[@n7.then, @n10.then_0]
    Pawn:Fort Player Pawn Object Reference=Get Pawn() -> (Pawn:Fort Player Pawn Object Reference)
    StatTag:Fort Stat Manager Tag Structure=(TagName="StatManager.AthenaKills")
    DefaultValue:int=0
    bBroadcastInitialValue:bool=true
    InitialValueBroadcastDelay:float=-1
    ChangeValueBroadcastDelay:float=-1
    bConsumeWaitingNotifications:bool=true
   ) -> (then:Exec, OnStatChanged:Exec, StatName:Name, StatValue:int) {is=title K2Node_AsyncAction}
   n9 = Set KillCount(execute=@n8.OnStatChanged, KillCount:double="To Float (Integer)"(InInt=@n8.StatValue)) -> (Output_Get:double)
   > Branch(Condition="float > float"(A=@n9, B=Get LastKillCount() -> (LastKillCount:double)) {is=title K2Node_CallFunction})
   > Set ActiveKill(ActiveKill=true)
   > Set LastKillCount(LastKillCount:double=@n9) -> (Output_Get:double)
   n10 = Sequence()
  ]}
]
variables [
 ActiveKill {type=boolean}
 didKill {type=number}
 KillCount {type=number}
 LastKillCount {type=number}
 Pawn {type=FortPlayerPawn}
 petalLowerRotate {type=Rotator}
 petalUpperRotate {type=Rotator}
 timer {type=number}
 timerLen {type=number, default=60}
]
